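# Physics asset for the Tech Explorer backpack cosmetic's grenade variant, providing collision and joint limits for its dangling grenade.
# Three capsule bodies cover the pack, a small ring bone and the grenade bone.
# Angular constraints chain grenade to ring to pack, with tight swing at the ring, wide swing at the grenade and limited twist on both joints.
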PhysicsAsset /BRCosmetics/Accessories/FORT_Backpacks/M_MED_Tech_Explorer_Backpack/Meshes/M_MED_Tech_Explorer_Pack_Nade_Physics
bodies {
 dyn_nade [{shapeType=Capsule, length=4.5, radius=3.25, center=<4.1523 -1.6672 -1.2593>, rotation=<rot 81.9741 -5.0513 4.6589>}]
 dyn_ring [{shapeType=Capsule, length=1, radius=2.75, center=<2.4609 0 0>, rotation=<rot 90 0 0>}]
 pack [{shapeType=Capsule, length=21.75, radius=8.75, center=<0 8.2512 -23.2818>, rotation=<rot 0 0 10.0365>}]
}
constraints [
 {bone1Name=dyn_ring, bone2Name=pack, swing1LimitDegrees=20, swing2LimitDegrees=20, twistLimitDegrees=8}
 {bone1Name=dyn_nade, bone2Name=dyn_ring, swing1LimitDegrees=90, swing2LimitDegrees=90, twistLimitDegrees=12}
]
