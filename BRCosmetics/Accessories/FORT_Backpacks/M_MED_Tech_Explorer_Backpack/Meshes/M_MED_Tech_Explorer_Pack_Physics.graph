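# Physics asset for the Tech Explorer backpack cosmetic, defining how its dangling parts swing and collide.
# Capsule bodies cover the main pack and small dynamic bones for a hoop, two pouches, two scope segments and a zipper.
# Angular-limited constraints form short chains from the pack: scope to scope, pouch to pouch with the hoop branching off the first pouch, and the zipper directly attached.
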PhysicsAsset /BRCosmetics/Accessories/FORT_Backpacks/M_MED_Tech_Explorer_Backpack/Meshes/M_MED_Tech_Explorer_Pack_Physics
bodies {
 dyn_hoop [{shapeType=Capsule, length=1.5, radius=1.25, center=<0.8 0 0>, rotation=<rot 90 0 0>}]
 dyn_pouch_01 [{shapeType=Capsule, length=2.25, radius=3.5, center=<-0.2 0 0>, rotation=<rot 90 0 0>}]
 dyn_pouch_02 [{shapeType=Capsule, length=3.25, radius=3.25, center=<-0.35 0 0>, rotation=<rot 90 0 0>}]
 dyn_scope_01 [{shapeType=Capsule, length=1, radius=1, center=<0 0 -0.25>}]
 dyn_scope_02 [{shapeType=Capsule, length=1, radius=1, center=<0 0 -0.75>}]
 dyn_zipper [{shapeType=Capsule, length=1, radius=0.75, center=<0 1.1423 0>, rotation=<rot 0 0 90>}]
 pack [{shapeType=Capsule, length=21.5, radius=9, center=<0 8.2512 -23.2818>, rotation=<rot 0 0 11.0866>}]
}
constraints [
 {bone1Name=dyn_scope_01, bone2Name=pack, swing1LimitDegrees=5, swing2LimitDegrees=15, twistLimitDegrees=3}
 {bone1Name=dyn_scope_02, bone2Name=dyn_scope_01, swing1LimitDegrees=5, swing2LimitDegrees=35, twistLimitDegrees=3}
 {bone1Name=dyn_pouch_01, bone2Name=pack, swing1LimitDegrees=11, swing2LimitDegrees=11, twistLimitDegrees=8}
 {bone1Name=dyn_pouch_02, bone2Name=dyn_pouch_01, swing1LimitDegrees=15, swing2LimitDegrees=15, twistLimitDegrees=13}
 {bone1Name=dyn_hoop, bone2Name=dyn_pouch_01, swing1LimitDegrees=8, swing2LimitDegrees=45, twistLimitDegrees=7}
 {bone1Name=dyn_zipper, bone2Name=pack, swing1LimitDegrees=25, swing2LimitDegrees=25, twistLimitDegrees=5}
]
